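# Physics asset for the Peaceful Poem (Pumped) guitar cosmetic, providing collision and joint limits for two dangling cloth pieces, left and right.
# Each side is a three-segment capsule chain hung from a small MSTR root body, with per-joint swing and twist limits that differ slightly between sides.
# The two roots are locked together in swing and twist, while an unconstrained handle capsule adds a larger body on the guitar.
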PhysicsAsset /SparksCosmetics/Instrument/Guitar/PeacefulPoem_Pumped/Meshes/PeacefulPoem_Pumped_Guitar_Physics
bodies {
 PeacefulPoem_dyn_cloth_01_l [{shapeType=Capsule, length=5, radius=1, center=<2.0833 0.0071 0>, rotation=<rot 90 180 179.8047>}]
 PeacefulPoem_dyn_cloth_01_l_MSTR [{shapeType=Capsule, length=0, radius=0.505}]
 PeacefulPoem_dyn_cloth_01_r [{shapeType=Capsule, length=5, radius=1, center=<2.0833 0.0071 0>, rotation=<rot 90 179.8694 179.6742>}]
 PeacefulPoem_dyn_cloth_01_r_MSTR [{shapeType=Capsule, length=0, radius=0.505}]
 PeacefulPoem_dyn_cloth_02_l [{shapeType=Capsule, length=3.9127, radius=1.2806, center=<2.0684 0.0071 -0.0441>, rotation=<rot 90 180 179.8047>}]
 PeacefulPoem_dyn_cloth_02_r [{shapeType=Capsule, length=5.2042, radius=1.2133, center=<2.3979 0.0082 -0.2233>, rotation=<rot 90 180 179.8047>}]
 PeacefulPoem_dyn_cloth_03_l [{shapeType=Capsule, length=2.3806, radius=1.6194, center=<1.1496 0.0039 0>, rotation=<rot 90 180 179.8047>}]
 PeacefulPoem_dyn_cloth_03_r [{shapeType=Capsule, length=1.9802, radius=2.0198, center=<0.8024 0.0027 0>, rotation=<rot 90 180 179.8047>}]
 handle [{
   shapeType=Capsule
   length=4.7051
   radius=4.6327
   shapeName=handle_capsule
   center=<-3.2852 -1.7181 -15.6194>
   rotation=<rot 0.1046 0 -44.0936>
  }]
}
constraints [
 {
  bone1Name=PeacefulPoem_dyn_cloth_01_l
  bone2Name=PeacefulPoem_dyn_cloth_01_l_MSTR
  swing1LimitDegrees=25
  swing2LimitDegrees=90
  twistLimitDegrees=5
 }
 {
  bone1Name=PeacefulPoem_dyn_cloth_02_l
  bone2Name=PeacefulPoem_dyn_cloth_01_l
  swing1LimitDegrees=20
  swing2LimitDegrees=50
  twistLimitDegrees=15
 }
 {
  bone1Name=PeacefulPoem_dyn_cloth_03_l
  bone2Name=PeacefulPoem_dyn_cloth_02_l
  swing1LimitDegrees=35
  swing2LimitDegrees=25
  twistLimitDegrees=23
 }
 {
  bone1Name=PeacefulPoem_dyn_cloth_03_r
  bone2Name=PeacefulPoem_dyn_cloth_02_r
  swing1LimitDegrees=25
  swing2LimitDegrees=25
  twistLimitDegrees=15
 }
 {
  bone1Name=PeacefulPoem_dyn_cloth_01_r
  bone2Name=PeacefulPoem_dyn_cloth_01_r_MSTR
  swing1LimitDegrees=25
  swing2LimitDegrees=90
  twistLimitDegrees=15
 }
 {
  bone1Name=PeacefulPoem_dyn_cloth_02_r
  bone2Name=PeacefulPoem_dyn_cloth_01_r
  swing1LimitDegrees=30
  swing2LimitDegrees=50
  twistLimitDegrees=15
 }
 {
  bone1Name=PeacefulPoem_dyn_cloth_01_r_MSTR
  bone2Name=PeacefulPoem_dyn_cloth_01_l_MSTR
  swing1LimitDegrees=45
  swing1Motion=Locked
  swing2LimitDegrees=45
  swing2Motion=Locked
  twistLimitDegrees=45
  twistMotion=Locked
 }
]
